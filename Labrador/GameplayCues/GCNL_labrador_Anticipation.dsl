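; Looping gameplay cue that shows an anticipation effect on a Labrador AI pawn by swapping in a piñata anticipation material.
; On looping start it casts the target to FortAIPawn, then creates and applies a dynamic material instance to mesh slot 0.
; The cue's raw magnitude sets the effect length and current game time sets its start stamp, leaving animation to the material.
; No logic here restores the original material.
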(event Collision|EventActorBeginOverlap
  (params (OtherActor :type Actor%20Object%20Reference)))

(event EventTick
  (params (DeltaSeconds :type Float%20%28single-precision%29)))

(event EventOnLoopingStart(EitherVFX-preferthisversion)
  (params (MyTarget :type Actor%20Object%20Reference) (Parameters :type Gameplay%20Cue%20Parameters%20Structure%20%28by%20ref%29) (ParticleComponents :type Array%20of%20FXSystem%20Component%20Object%20References) (AudioComponents :type Array%20of%20Audio%20Component%20Object%20References))
  (bind (_normalizedmagnitude _rawmagnitude) (Ability|GameplayCue|BreakGameplayCueParameters Parameters))
  (bind _mid_anticipation (Variables|Default|GetMID_Anticipation))
  (bind _asfort_aipawn (Utilities|Casting|CastToFortAIPawn MyTarget))
  (bind _returnvalue (Rendering|Material|CreateDynamicMaterialInstance (Variables|/Script/Engine.Character|GetMesh _asfort_aipawn) 0 "/Labrador/Meshes/M_Labrador_Pinata_Anticipation.M_Labrador_Pinata_Anticipation" "Anticipation_MID" :!class /Script/Engine.PrimitiveComponent))
  (bind _output_get (Variables|Default|SetMID_Anticipation _returnvalue))
  (Rendering|Material|SetMaterial (Variables|/Script/Engine.Character|GetMesh _asfort_aipawn) 0 _output_get)
  (Rendering|Material|SetScalarParameterValue _mid_anticipation "AntiTimeLength" _rawmagnitude :!class /Script/Engine.MaterialInstanceDynamic)
  (Rendering|Material|SetScalarParameterValue _mid_anticipation "AntiTimeStamp" (Utilities|Time|GetGameTimeinSeconds 0) :!class /Script/Engine.MaterialInstanceDynamic))

(fn ConstructionScript)
