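; Latent burst gameplay cue that plays the golden loot-leak effect for Labrador on a spawned pickup.
; A zero-second delay works around a race between server pickup spawn and client loading of mesh components.
; It attaches a Niagara system to the pickup's skeletal or static mesh, falling back to the root component.
; It also feeds the pickup's velocity, transformed and normalized, into a burst particle component's "Loot Direction" parameter.
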(event EventOnBurst(EitherVFX-preferthisversion)
  (params (MyTarget :type Actor%20Object%20Reference) (Parameters :type Gameplay%20Cue%20Parameters%20Structure%20%28by%20ref%29) (ParticleComponents :type Array%20of%20FXSystem%20Component%20Object%20References) (AudioComponents :type Array%20of%20Audio%20Component%20Object%20References) (BurstCameraShakeInstance :type Camera%20Shake%20Base%20Object%20Reference) (BurstDecalInstance :type Decal%20Actor%20Object%20Reference))
  (bind (_normalizedmagnitude _rawmagnitude _effectcontext _matchedtagname _originaltag _aggregatedsourcetags _aggregatedtargettags _location _normal _instigator _effectcauser _sourceobject) (Ability|GameplayCue|BreakGameplayCueParameters Parameters))
  ;[0 0 0] I hate this but there's a race condition with the server spawning the pickup and the client loading the effects blueprint which has the mesh components.
  (Utilities|FlowControl|Delay 0)
  (bind _asfort_pickup (Utilities|Casting|CastToFortPickup _sourceobject))
  (Niagara|SpawnSystemAttached (Variables|Default|GetPickUpFX_Template) (select (Class|FortPickup|IsSkeletalMeshComponent _asfort_pickup) (select (Utilities|IsValid (Class|FortPickup|GetPickupSkeletalMeshComponent _asfort_pickup) :!class /Script/Engine.KismetSystemLibrary) (Class|FortPickup|GetPickupSkeletalMeshComponent _asfort_pickup) (Variables|/Script/Engine.Actor|GetRootComponent _asfort_pickup)) (select (Utilities|IsValid (Class|FortPickup|GetPickupStaticMeshComponent _asfort_pickup) :!class /Script/Engine.KismetSystemLibrary) (Class|FortPickup|GetPickupStaticMeshComponent _asfort_pickup) (Variables|/Script/Engine.Actor|GetRootComponent _asfort_pickup))) "root" "0, 0, 0" "0, 0, 0" "SnapToTarget" true true "AutoRelease" false)
  (Effects|Components|ParticleSystem|SetVectorParameter (Utilities|Array|Get(acopy) ParticleComponents) "Loot Direction" (Math|Vector|Normalize (Math|Transform|TransformDirection (Transformation|GetWorldTransform (Utilities|Array|Get(acopy) ParticleComponents)) (Variables|/Script/Engine.MovementComponent|GetVelocity (Variables|/Script/FortniteGame.FortPickup|GetMovementComponent _asfort_pickup))))))

(fn ConstructionScript)
