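; Burst gameplay cue for the rez-out of a golden Labrador NPC pawn, dissolving its body with added VFX.
; On burst it applies the dissolve material, plays a rate-adjusted timeline, spawns Niagara linked to the skeletal mesh, and hides child FX components.
; A vertical trace detects water, spawning a surface splash scaled by mesh bounds.
; Empty events act as hooks; socket-math and non-mobile point-light helpers exist but are not called here.
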(event EventOnBurst(EitherVFX-preferthisversion)
  (params (MyTarget :type Actor%20Object%20Reference) (Parameters :type Gameplay%20Cue%20Parameters%20Structure%20%28by%20ref%29) (ParticleComponents :type Array%20of%20FXSystem%20Component%20Object%20References) (AudioComponents :type Array%20of%20Audio%20Component%20Object%20References) (BurstCameraShakeInstance :type Camera%20Shake%20Base%20Object%20Reference) (BurstDecalInstance :type Decal%20Actor%20Object%20Reference))
  (bind _asnpc_pawn_parent (Utilities|Casting|CastToNPC_Pawn_Parent MyTarget))
  (Variables|DONOTEDIT|SetLabrador _asnpc_pawn_parent)
  (CallFunction|Pre-SequenceChange)
  (CallFunction|SetLabradorMaterials)
  (CallFunction|WaterTrace)
  (CallFunction|SetTimelinePlayrates)
  (Components|Timeline|PlayfromStart (Variables|GCN_Labrador_RezOut_Golden|GetTFX_ResOutChracterMesh))
  (CallFunction|SpawnDissolveVFX)
  (CallFunction|SpawnWaterEliminationFX)
  (for _array_element (Components|GetChildrenComponents (Customization|GetSkeletalMeshforPartType (Variables|DONOTEDIT|GetLabrador) "Body" :!class /Script/FortniteGame.FortAIPawn) true)
    (bind _asfxsystem_component (Utilities|Casting|CastToFXSystemComponent _array_element))
    (Development|SetHiddeninGame _asfxsystem_component true true)
    (if (Variables|DEBUG|GetDEBUG_FXSYSTEMCOMPONENTS)
      (Development|PrintString 0 (Utilities|GetDisplayName _array_element)))))

(event Custom|Pre-Sequence%20Change)

(event Custom|Set%20Additional%20Niagara%20Parameters%20on%20Spawned%20FX
  (params (Spawned%20Death%20FX :type Niagara%20Particle%20System%20Component%20Object%20Reference)))

(fn Clean-UpTeleportationLight)

(fn SetLabradorMaterials
  (bind _labrador (Variables|DONOTEDIT|GetLabrador)
    (:then
      (for _index (range (+ (Utilities|Array|Length (Rendering|Material|GetMaterials (Customization|GetSkeletalMeshforPartType _labrador "Body" :!class /Script/FortniteGame.FortAIPawn))) 1))
        (Rendering|Material|SetMaterial (Customization|GetSkeletalMeshforPartType _labrador "Body" :!class /Script/FortniteGame.FortAIPawn) _index (Variables|DISSOLVE|GetMat%20Chracter%20Dissolve)))
      (return))
    (:else)))

(fn SetTimelinePlayrates
  (Components|Timeline|SetPlayRate (Variables|GCN_Labrador_RezOut_Golden|GetTFX_ResOutChracterMesh) (Variables|DISSOLVE|GetDissolve%20Timeline%20Playrate)))

(fn SocketMath
  (bind _labrador (Variables|DONOTEDIT|GetLabrador)
    (:then
      (return (Math|Vector|VectorLength (- (+ (Transformation|GetSocketLocation (Customization|GetSkeletalMeshforPartType _labrador "Body" :!class /Script/FortniteGame.FortAIPawn) (Variables|DISSOLVE|GetSocket%20Mesh%20Bottom)) (* (Math|Vector|Normalize (- (Transformation|GetSocketLocation (Customization|GetSkeletalMeshforPartType _labrador "Body" :!class /Script/FortniteGame.FortAIPawn) (Variables|DISSOLVE|GetSocket%20Mesh%20Top)) (Transformation|GetSocketLocation (Customization|GetSkeletalMeshforPartType _labrador "Body" :!class /Script/FortniteGame.FortAIPawn) (Variables|DISSOLVE|GetSocket%20Mesh%20Bottom)))) -1e+01)) (+ (Transformation|GetSocketLocation (Customization|GetSkeletalMeshforPartType _labrador "Body" :!class /Script/FortniteGame.FortAIPawn) (Variables|DISSOLVE|GetSocket%20Mesh%20Top)) (* (Math|Vector|Normalize (- (Transformation|GetSocketLocation (Customization|GetSkeletalMeshforPartType _labrador "Body" :!class /Script/FortniteGame.FortAIPawn) (Variables|DISSOLVE|GetSocket%20Mesh%20Top)) (Transformation|GetSocketLocation (Customization|GetSkeletalMeshforPartType _labrador "Body" :!class /Script/FortniteGame.FortAIPawn) (Variables|DISSOLVE|GetSocket%20Mesh%20Bottom)))) 4e+01)))) (+ (Transformation|GetSocketLocation (Customization|GetSkeletalMeshforPartType _labrador "Body" :!class /Script/FortniteGame.FortAIPawn) (Variables|DISSOLVE|GetSocket%20Mesh%20Bottom)) (* (Math|Vector|Normalize (- (Transformation|GetSocketLocation (Customization|GetSkeletalMeshforPartType _labrador "Body" :!class /Script/FortniteGame.FortAIPawn) (Variables|DISSOLVE|GetSocket%20Mesh%20Top)) (Transformation|GetSocketLocation (Customization|GetSkeletalMeshforPartType _labrador "Body" :!class /Script/FortniteGame.FortAIPawn) (Variables|DISSOLVE|GetSocket%20Mesh%20Bottom)))) -1e+01)) (* (Math|Vector|Normalize (- (Transformation|GetSocketLocation (Customization|GetSkeletalMeshforPartType _labrador "Body" :!class /Script/FortniteGame.FortAIPawn) (Variables|DISSOLVE|GetSocket%20Mesh%20Top)) (Transformation|GetSocketLocation (Customization|GetSkeletalMeshforPartType _labrador "Body" :!class /Script/FortniteGame.FortAIPawn) (Variables|DISSOLVE|GetSocket%20Mesh%20Bottom)))) -1.0)))
    (:else
      (return 0.0 "0, 0, 0" "0, 0, 0"))))

(fn SpawnDissolveVFX
  (if (Variables|DEBUG|GetDEBUG_REMOVESPAWNVFX)
    (Development|PrintString 0 "DEBUG - Removed Spawned VFX")
    (return)
    (else
      (bind _labrador (Variables|DONOTEDIT|GetLabrador)
        (:then
          (bind _returnvalue (Niagara|SpawnSystematLocation 0 (Variables|DISSOLVE|GetDissolve%20Visual%20Effect) (Transformation|GetSocketLocation (Customization|GetSkeletalMeshforPartType _labrador "Body" :!class /Script/FortniteGame.FortAIPawn) (Variables|DISSOLVE|GetDissolve%20VFX%20Spawn%20Point%20Name))))
          (bind _output_get (Variables|DONOTEDIT|SetSpawned%20Death%20VFX _returnvalue))
          (Niagara|SetNiagaraSkeletalMeshComponent _output_get "SkeletalMesh" (Customization|GetSkeletalMeshforPartType (Variables|DONOTEDIT|GetLabrador) "Body" :!class /Script/FortniteGame.FortAIPawn))
          (Niagara|SetNiagaraVariable(Vector3) _output_get "PointPosition" "0, 0,250.000000")
          (CallFunction|SetAdditionalNiagaraParametersOnSpawnedFX (Variables|DONOTEDIT|GetSpawned%20Death%20VFX))
          (return))
        (:else)))))

(fn SpawnTeleportationLight
  (bind _labrador (Variables|DONOTEDIT|GetLabrador)
    (:then
      (if (Variables|DEBUG|GetDEBUG_REMOVESPAWNEDPOINTLIGHT)
        _
        (elif (Fortnite|IsMobilePlatform)
          (Development|PrintString (Variables|DONOTEDIT|GetLabrador) "Rez Out - MOBILE PLATFORM - No Point LIght for you.")
          (else
            (AddComponent|NODE_AddPointLightComponent-0 :RelativeTransform (Math|Conversions|ToTransform(Vector) (+ (Transformation|GetSocketLocation (Customization|GetSkeletalMeshforPartType _labrador "Body" :!class /Script/FortniteGame.FortAIPawn) "root") "0, 0, 0")))))))
    (:else)))

(fn SpawnWaterEliminationFX
  (bind _water_hit_information (Variables|WaterDissolve|GetWater%20Hit%20Information))
  (bind (_bblockinghit _binitialoverlap _time _distance _location _impactpoint _normal _impactnormal _physmat _hitactor) (Collision|BreakHitResult _water_hit_information))
  (if (Variables|WaterDissolve|GetWater%20Hit)
    (bind _asfort_water_body_actor (Utilities|Casting|CastToFortWaterBodyActor _hitactor))
    (bind _outwatersurfaceinfo (WaterBody|GetWaterSurfaceInfoStruct _asfort_water_body_actor _location true false false))
    (bind _returnvalue (Niagara|SpawnSystematLocation (Variables|DONOTEDIT|GetLabrador) (Variables|WaterDissolve|GetWater%20Elimination%20Splash) (Utilities|Struct|BreakWaterSurfaceInfo _outwatersurfaceinfo) "90.000000, 0, 0"))
    (Effects|Components|ParticleSystem|SetVectorParameter _returnvalue "WaterPlaneLocation" (Utilities|Struct|BreakWaterSurfaceInfo _outwatersurfaceinfo))
    (Effects|Components|ParticleSystem|SetVectorParameter _returnvalue "WaterPlaneNormal" (.WaterPlaneNormal (Utilities|Struct|BreakWaterSurfaceInfo _outwatersurfaceinfo)))
    (Effects|Components|ParticleSystem|SetFloatParameter _returnvalue "WaterDepth" (.WaterDepth (Utilities|Struct|BreakWaterSurfaceInfo _outwatersurfaceinfo)))
    (Effects|Components|ParticleSystem|SetFloatParameter _returnvalue "WaterBodyIndex" (Math|Conversions|ToFloat(Integer) (.WaterBodyIdx (Utilities|Struct|BreakWaterSurfaceInfo _outwatersurfaceinfo))))
    (Effects|Components|ParticleSystem|SetFloatParameter _returnvalue "BoundScale" (Variables|WaterDissolve|GetBounds%20Scalar))
    (return)
    (else
      (return))))

(fn ConstructionScript)

(fn WaterTrace
  (bind _labrador (Variables|DONOTEDIT|GetLabrador)
    (:then
      (Variables|WaterDissolve|SetBounds%20Scalar (/ (.SphereRadius (Collision|GetComponentBounds (Customization|GetSkeletalMeshforPartType _labrador "Body" :!class /Script/FortniteGame.FortAIPawn))) 130.5))
      (bind (_outhit _returnvalue) (Collision|LineTraceForObjects 0 (+ (Transformation|GetActorLocation (Variables|DONOTEDIT|GetLabrador)) "0, 0,100.000000") (- (Transformation|GetActorLocation (Variables|DONOTEDIT|GetLabrador)) "0, 0,100.000000") (Utilities|Array|MakeArray "ObjectTypeQuery12")))
      (if _returnvalue
        (Variables|WaterDissolve|SetWater%20Hit%20Information _outhit)
        (else
          _))
      (Variables|WaterDissolve|SetWater%20Hit _returnvalue)
      (return))
    (:else
      ;[1 1 1] Can't have hit when the trace never happened
      (Variables|WaterDissolve|SetWater%20Hit false))))
